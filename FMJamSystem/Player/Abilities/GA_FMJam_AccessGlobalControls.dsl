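; Gameplay ability that opens the Jam global controls UI for the activating player.
; Server-gated activation asks the jam session whether access is allowed, then either adds a Dynamic UI scene, chosen by game-state class (STW variant or default), or reports a tagged failure, also when the pawn is hidden.
; An event router listener ends the ability when the controls report not displayed.
; Ending removes the scene on clients and stops listening.
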(event Ability|EventActivateAbility
  (bind _returnvalue (Ability|GetActivatingPawn))
  (if (Class|JamCoreBPFL|ShouldRunServerCode _returnvalue)
    (bind _returnvalue_2 (Mojo|JamSession|GetJamSessionforActor _returnvalue))
    (if (Utilities|IsValidInterface _returnvalue_2)
      (if (.ReturnValue (Class|JamSessionInterface|ShouldAllowGlobalControlAccess _returnvalue_2 (Mojo|JamSession|GetNetIdforPawn _returnvalue_2 _returnvalue)))
        (bind _bdidactorhavetags (Class|AthenaQuestFunctionLibrary|AthenaCheckActorHasTags (Ability|GetActivatingPawn) "(GameplayTags=((TagName=\"Pawn.HidePlayerPawn.Default\")))"))
        (if _bdidactorhavetags
          (CallFunction|Error (Variables|Default|GetGlobalControlsBusyTag))
          (else
            (CallFunction|AddGCScene)))
        (else
          (CallFunction|Error (Class|JamSessionInterface|ShouldAllowGlobalControlAccess _returnvalue_2 (Mojo|JamSession|GetNetIdforPawn _returnvalue_2 _returnvalue))))))))

(event Custom|AddGCScene
  (bind _returnvalue (Ability|GetActivatingPawn))
  (bind _returnvalue_2 (Fortnite|GetFortPlayerControllerfromActor _returnvalue))
  (DynamicUI|AddScene (WorldSubsystems|GetDynamicUIManager) (CallFunction|GetGlobalControlsSceneforGameMode) _returnvalue_2)
  (bind _returnvalue_3 (Mojo|JamSession|GetJamSessionforActor _returnvalue_2))
  (if (Utilities|IsValidInterface _returnvalue_3)
    (Utilities|IsValid (Class|JamSessionInterface|GetGameplayEventRouterComponent _returnvalue_3)
      (:Is%20Valid
        (bind (_asynctaskproxy _payload) (GameplayEventRouter|StartListeningToEvent (Class|JamSessionInterface|GetGameplayEventRouterComponent _returnvalue_3) "/FMJamSystem/Widgets/GlobalControls/JamEvent_GlobalControls_Displayed.JamEvent_GlobalControls_Displayed" _returnvalue_2)
          (:then
            (Variables|Default|SetAsyncRefGlobalControlsDisplayed _asynctaskproxy))
          (:OnEventReceived
            (if (not (Utilities|Struct|BreakJamEventGlobalControlsDisplayed _payload))
              (Ability|EndAbility))))))))

(event Custom|Error
  (params (ErrorReason :type Gameplay%20Tag%20Container%20Structure))
  (bind _asfort_player_pawn (Utilities|Casting|CastToFortPlayerPawn (Ability|GetActivatingPawn))
    (:then
      (FailedAction|UnabletoPerformAction _asfort_player_pawn true ErrorReason))
    (:CastFailed
      _))
  (Ability|EndAbility))

(event Ability|EventOnEndAbility
  (params (bWasCancelled :type Boolean))
  (Utilities|FlowControl|Sequence
    (:then_0
      (if (Class|JamCoreBPFL|ShouldRunClientCode (Ability|GetActivatingPawn))
        (CallFunction|RemoveGCScene)))
    (:then_1
      (GameplayEventRouter|StopListeningtoEvent (Variables|Default|GetAsyncRefGlobalControlsDisplayed)))
    (:then_2
      (|CallParentFunction bWasCancelled))))

(event Custom|RemoveGCScene
  (DynamicUI|RemoveScene (WorldSubsystems|GetDynamicUIManager) (CallFunction|GetGlobalControlsSceneforGameMode) (Fortnite|GetFortPlayerControllerfromActor (Ability|GetActivatingPawn))))

(fn GetGlobalControlsSceneForGameMode
  (return (select (Utilities|ClassIsChildOf (Utilities|GetClass (Game|GetGameState)) (Variables|Default|GetSTW%20Parent%20Class)) "/FMJamSystem/Scenes/Jam_GlobalControlsCenter.Jam_GlobalControlsCenter" "/FMJamSystem/Scenes/Jam_GlobalControls.Jam_GlobalControls")))
