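; GA_FMJam_DuckSfx is a gameplay ability in the FMJam system that, going by its name, ducks sound effects during a jam.
; On activation it calls the parent, then activates the CBM_JamDuckSfx audio modulation control bus mix.
; When the ability ends, cancelled or not, it calls the parent and deactivates that mix, so the mix lasts as long as the ability.
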(event Ability|EventActivateAbility
  (|CallParentFunction)
  (Audio|Modulation|ActivateControlBusMix "/FMJamSystem/Audio/CBM_JamDuckSfx.CBM_JamDuckSfx"))

(event Ability|EventOnEndAbility
  (params (bWasCancelled :type Boolean))
  (|CallParentFunction bWasCancelled)
  (Audio|Modulation|DeactivateControlBusMix "/FMJamSystem/Audio/CBM_JamDuckSfx.CBM_JamDuckSfx"))
