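; Gameplay ability in the FM Jam system that governs visibility of the Global Controls input prompt, hiding it during Dance Spotlight to avoid a conflict.
; Its CanActivateAbility override gets the Ability System Component from the actor info and refuses activation while it has the FMJam.Alias.InhibitElement.GlobalControls tag.
; Activation is allowed when the tag is absent or the component is invalid.
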(fn CanActivateAbility
  (params (ActorInfo :type Gameplay%20Ability%20Actor%20Info%20Structure) (Handle :type Gameplay%20Ability%20Spec%20Handle%20Structure))
  (bind _actorinfo (Variables|GetActorInfo))
  (bind (_owneractor _avataractor _playercontroller _abilitysystemcomponent) (Utilities|Struct|BreakGameplayAbilityActorInfo _actorinfo))
  ; Hide the Global Controls input prompt if the player is in Dance Spotlight to avoid a conflict
  (Utilities|IsValid _abilitysystemcomponent
    (:Is%20Valid
      (return 0 (not (GameplayTags|HasMatchingGameplayTag _abilitysystemcomponent "(TagName=\"FMJam.Alias.InhibitElement.GlobalControls\")"))))
    (:Is%20Not%20Valid
      (return 0 true))))
