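; Gameplay ability in the FM Jam system that stops the player's own jam loop.
; Activation is allowed only when the avatar's jam session music manager has an in-use music slot for that player's net ID.
; On activation it finds the BP_JamControllerComponent on the pawn's controller, calls StopLoop, then immediately ends the ability.
; CanApplyHolster always returns false.
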(event Ability|EventOnEndAbility
  (params (bWasCancelled :type Boolean)))

(event Ability|EventActivateAbility
  (Utilities|FlowControl|Sequence
    (:then_0
      (Utilities|IsValid (Ability|GetActivatingPawn)
        (:Is%20Valid
          (bind _output_get (Variables|Default|SetThe%20Controller (Pawn|GetController (Ability|GetActivatingPawn))))
          (Utilities|IsValid _output_get
            (:Is%20Valid
              (bind _output_get_2 (Variables|Default|SetThe%20JamControllerComponent (Actor|GetComponentbyClass (Variables|Default|GetThe%20Controller) "/FMJamSystem/Components/BP_JamControllerComponent.BP_JamControllerComponent_C")))
              (Utilities|IsValid _output_get_2
                (:Is%20Valid
                  (Class|BPJamControllerComponent|StopLoop (Variables|Default|GetThe%20JamControllerComponent)))))))))
    (:then_1
      (Ability|EndAbility))))

(fn CanApplyHolster
  (params (HolsterId :type Name%20%28by%20ref%29 :default "None"))
  (return false))

(fn CanActivateAbility
  (params (ActorInfo :type Gameplay%20Ability%20Actor%20Info%20Structure) (Handle :type Gameplay%20Ability%20Spec%20Handle%20Structure))
  (bind _actorinfo (Variables|GetActorInfo))
  (bind (_owneractor _avataractor) (Utilities|Struct|BreakGameplayAbilityActorInfo _actorinfo))
  (bind _jamsession (LocalVariables|GetJamSession))
  (bind _returnvalue (Class|JamSessionInterface|GetMusicManager _jamsession))
  ; We want players to only be able to stop their base Jam if they're, well, ykno, jamming.
  (Utilities|FlowControl|Sequence
    (:then_0
      (LocalVariables|SetPawn (Utilities|Casting|CastToFortPawn _avataractor))
      (bind _returnvalue_2 (Mojo|JamSession|GetJamSessionforActor _avataractor))
      (LocalVariables|SetJamSession _returnvalue_2))
    (:then_1
      (bind _pawn (LocalVariables|GetPawn)
        (:then
          (LocalVariables|SetPlayerState (Variables|/Script/Engine.Pawn|GetPlayerState _pawn)))
        (:else)))
    (:then_2
      (if (and (Utilities|IsValid (LocalVariables|GetPlayerState) :!class /Script/Engine.KismetSystemLibrary) (Utilities|IsValidInterface (LocalVariables|GetJamSession)))
        (Utilities|IsValid _returnvalue
          (:Is%20Valid
            (LocalVariables|SetMusicSlot (Mojo|JamSession|GetMusicSlotforPlayerId _returnvalue (PlayerState|GetUniqueNetId (LocalVariables|GetPlayerState))))))))
    (:then_3
      ;; Can stop their jam if they have a jam to stop
      (bind _musicslot (LocalVariables|GetMusicSlot)
        (:then
          (if (Class|JamMusicSlot|IsinUse _musicslot)
            (return 0 true)))
        (:else)))))
